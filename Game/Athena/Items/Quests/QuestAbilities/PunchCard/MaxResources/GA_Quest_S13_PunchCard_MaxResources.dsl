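; Passive quest ability that credits the Season 13 punch card quest for holding max wood, stone and metal.
; For the first local player, it waits for quests to be ready, then binds pickup and harvest events if the objective is incomplete.
; Each event checks all three counts against a threshold, with harvests delayed briefly.
; On success it sends the MaxResources stat event through ServerStatEvent and unbinds both listeners.
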(event Ability|EventActivateAbility
  (bind _owningplayerpawn (Variables|GAVars|Data|GetOwningPlayerPawn))
  (bind _returnvalue (Fortnite|GetFortPlayerControllerfromActor _owningplayerpawn))
  (bind _resourcesquest (Variables|Default|GetResourcesQuest))
  (|CallParentFunction)
  (if (Utilities|Equal(Object) (Fortnite|GetFirstLocalFortPlayerController true) (Fortnite|GetFortPlayerControllerfromActor (Variables|GAVars|Data|GetOwningPlayerPawn)))
    (Quests|CurrentQuestsReadyAsync (Variables|GAVars|Data|GetOwningPlayerQuestManager) (Utilities|Array|MakeArray (Variables|Default|GetResourcesQuest))
      (:OnCurrentQuestsReady
        (bind _questvalid (Quest|DoesPlayerHaveQuest _returnvalue _resourcesquest "Quest_S13_PunchCard_MaxResources"))
        (bind (_playercontrollerout _completedquestobjective) (Quest|HasPlayerCompletedQuestObjective _returnvalue _resourcesquest "Quest_S13_PunchCard_MaxResources"))
        (if (and _questvalid (not _completedquestobjective))
          (Inventory|BindEventtoOnItemPickedUp (BlueprintContext|GetFortInventoryContext) (delegate Custom|ItemPickedUp))
          (HUD|BindEventtoOnDamagedResourceBuilding (BlueprintContext|GetFortHUDContext) (delegate Custom|ResourcesHarvested)))))))

(event Ability|EventFailedToActivatePassiveAbility
  (params (ActorInfo :type Gameplay%20Ability%20Actor%20Info%20Structure)))

(event Custom|ItemPickedUp
  (params (NewItem :type Fort%20World%20Item%20Object%20Reference) (Count :type Integer))
  (CallFunction|CheckPlayerResources))

(event Custom|ResourcesHarvested
  (params (BuildingSMActor :type Building%20Static%20Mesh%20Object%20Reference) (PotentialWorldItem :type Fort%20Item%20Object%20Reference) (bDestroyed :type Boolean) (bJustHitWeakspot :type Boolean))
  (Utilities|FlowControl|Delay 0.1)
  (CallFunction|CheckPlayerResources))

(event Custom|ServerStatEvent
  (Quest|SendComplexCustomStatEvent (Variables|GAVars|Data|GetOwningPlayerQuestManager) 0 "(GameplayTags=)" "(GameplayTags=((TagName=\"Athena.Quests.MaxResources\")))" 1))

(fn CheckPlayerResources
  (if (Math|Boolean|ANDBoolean (>= (Inventory|GetResourceCount (BlueprintContext|GetFortInventoryContext) "Wood") (Variables|Default|GetResourceThreshold)) (>= (Inventory|GetResourceCount (BlueprintContext|GetFortInventoryContext) "Stone") (Variables|Default|GetResourceThreshold)) (>= (Inventory|GetResourceCount (BlueprintContext|GetFortInventoryContext) "Metal") (Variables|Default|GetResourceThreshold)))
    (CallFunction|ServerStatEvent)
    (Inventory|UnbindallEventsfromOnItemPickedUp (BlueprintContext|GetFortInventoryContext))
    (HUD|UnbindallEventsfromOnDamagedResourceBuilding (BlueprintContext|GetFortHUDContext))))
